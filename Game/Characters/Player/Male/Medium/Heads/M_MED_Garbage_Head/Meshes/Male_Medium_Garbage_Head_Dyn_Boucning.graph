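# Physics asset for the male medium player's garbage head, which by its name drives bouncing secondary motion on the head's dyn bones.
# Small capsule bodies cover the dyn handle, head, neck and spine bones, plus the regular head and spine_04 bones.
# Constraints form two chains meeting at spine_04: dyn_handle through dyn_head and head, and dyn_neck_01 through dyn_spine_04.
# The links into spine_04 explicitly free swing and twist.
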PhysicsAsset /Game/Characters/Player/Male/Medium/Heads/M_MED_Garbage_Head/Meshes/Male_Medium_Garbage_Head_Dyn_Boucning
bodies {
 dyn_handle [{shapeType=Capsule, length=2, radius=2}]
 dyn_head [{shapeType=Capsule, length=4.25, radius=2.5, rotation=<rot 90 0 0>}]
 dyn_neck_01 [{shapeType=Capsule, length=4.5, radius=2}]
 dyn_spine_04 [{shapeType=Capsule, length=0.505, radius=0.505, rotation=<rot 0 -90 0>}]
 head [{shapeType=Capsule, length=2, radius=0.5}]
 spine_04 [{shapeType=Capsule, length=2, radius=0.5}]
}
constraints [
 {bone1Name=dyn_handle, bone2Name=dyn_head, swing1LimitDegrees=45, swing2LimitDegrees=45, twistLimitDegrees=45}
 {bone1Name=dyn_head, bone2Name=head, swing1LimitDegrees=45, swing2LimitDegrees=45, twistLimitDegrees=45}
 {
  bone1Name=head
  bone2Name=spine_04
  swing1LimitDegrees=45
  swing1Motion=Free
  swing2LimitDegrees=45
  swing2Motion=Free
  twistLimitDegrees=45
  twistMotion=Free
 }
 {bone1Name=dyn_neck_01, bone2Name=dyn_spine_04, swing1LimitDegrees=45, swing2LimitDegrees=45, twistLimitDegrees=45}
 {
  bone1Name=dyn_spine_04
  bone2Name=spine_04
  swing1LimitDegrees=45
  swing1Motion=Free
  swing2LimitDegrees=45
  swing2Motion=Free
  twistLimitDegrees=45
  twistMotion=Free
 }
]
