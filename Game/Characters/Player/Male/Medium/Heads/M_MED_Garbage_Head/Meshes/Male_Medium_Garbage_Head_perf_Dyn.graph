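# Physics asset for the male medium Garbage head, setting up jointed secondary motion for its hoses and two note pieces.
# The top and bottom notes swing on one limited axis from dyn_head, which is locked to head.
# Three hoses branch from dyn_neck_01, locked to dyn_spine_04, which also roots a looser three-link neck-hose chain.
# Compound capsule-and-box shapes form dyn_spine_04; its joint to spine_04, like head's, rotates freely.
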PhysicsAsset /Game/Characters/Player/Male/Medium/Heads/M_MED_Garbage_Head/Meshes/Male_Medium_Garbage_Head_perf_Dyn
bodies {
 dyn_head [{
   shapeType=Capsule
   length=13.0707
   radius=10.48
   shapeName=dyn_head_capsule
   center=<3.0829 -1.4072 -0.1803>
   rotation=<rot -0.5011 -86.6804 -88.9252>
  }]
 dyn_hose_01 [{shapeType=Capsule, length=2, radius=2, center=<0.5 0 0>, rotation=<rot 90 0 0>}]
 dyn_hose_02 [{shapeType=Capsule, length=2, radius=2, center=<0.5 0 0>, rotation=<rot 90 0 0>}]
 dyn_hose_03 [{shapeType=Capsule, length=2, radius=2, center=<0.5 0 0>, rotation=<rot 90 0 0>}]
 dyn_neck_01 [{shapeType=Capsule, length=2, radius=2}]
 dyn_neck_hose_01 [{shapeType=Capsule, length=2.75, radius=1.25, center=<2 0 0>, rotation=<rot 90 0 0>}]
 dyn_neck_hose_02 [{shapeType=Capsule, length=2.25, radius=1.75, center=<2 0 0>, rotation=<rot 90 0 0>}]
 dyn_neck_hose_03 [{shapeType=Capsule, length=2.25, radius=1.75, center=<2 0 0>, rotation=<rot 90 0 0>}]
 dyn_note_bott [{
   shapeType=Capsule
   length=6.387
   radius=2.7029
   shapeName=dyn_note_bott_capsule
   center=<2.5838 0.9335 0.1426>
   rotation=<rot -2.051 -90 2.4242>
  }]
 dyn_note_top [{
   shapeType=Capsule
   length=4.0468
   radius=2.2441
   shapeName=dyn_note_top_capsule
   center=<2.6307 1.3689 0.3008>
   rotation=<rot -5.245 -89.9998 7.1007>
  }]
 dyn_spine_04 [
  {shapeType=Capsule, length=7.5, radius=4.75, center=<8.549 8.5168 -7.4513>, rotation=<rot -29.9993 10.0137 -1.4715>}
  {shapeType=Capsule, length=7.5, radius=4.75, center=<1.0084 8.4138 -11.515>, rotation=<rot -86.3219 29.5288 -25.6665>}
  {shapeType=Box, extentX=19.5, extentY=13.5, extentZ=24, center=<0.4727 6.0654 -2.0345>, rotation=<rot -59.367 3.082 0>}
 ]
 head [{shapeType=Capsule, length=2, radius=2}]
 spine_04 [{shapeType=Capsule, length=7.5, radius=1.1904, center=<1.7398 8.4625 -11.4725>, rotation=<rot -86.322 29.5285 -25.6667>}]
}
constraints [
 {
  bone1Name=dyn_note_top
  bone2Name=dyn_head
  swing1LimitDegrees=20
  swing2LimitDegrees=2
  swing2Motion=Locked
  twistLimitDegrees=2
  twistMotion=Locked
 }
 {
  bone1Name=dyn_note_bott
  bone2Name=dyn_head
  swing1LimitDegrees=20
  swing2LimitDegrees=2
  swing2Motion=Locked
  twistLimitDegrees=2
  twistMotion=Locked
 }
 {bone1Name=dyn_hose_01, bone2Name=dyn_neck_01, swing1LimitDegrees=15, swing2LimitDegrees=15, twistLimitDegrees=3}
 {bone1Name=dyn_neck_hose_01, bone2Name=dyn_spine_04, swing1LimitDegrees=45, swing2LimitDegrees=45, twistLimitDegrees=5}
 {bone1Name=dyn_neck_hose_02, bone2Name=dyn_neck_hose_01, swing1LimitDegrees=45, swing2LimitDegrees=45, twistLimitDegrees=5}
 {bone1Name=dyn_neck_hose_03, bone2Name=dyn_neck_hose_02, swing1LimitDegrees=45, swing2LimitDegrees=45, twistLimitDegrees=5}
 {bone1Name=dyn_hose_03, bone2Name=dyn_neck_01, swing1LimitDegrees=15, swing2LimitDegrees=15, twistLimitDegrees=3}
 {bone1Name=dyn_hose_02, bone2Name=dyn_neck_01, swing1LimitDegrees=15, swing2LimitDegrees=15, twistLimitDegrees=3}
 {
  bone1Name=dyn_head
  bone2Name=head
  swing1LimitDegrees=45
  swing1Motion=Locked
  swing2LimitDegrees=45
  swing2Motion=Locked
  twistLimitDegrees=45
  twistMotion=Locked
 }
 {
  bone1Name=head
  bone2Name=spine_04
  swing1LimitDegrees=45
  swing1Motion=Free
  swing2LimitDegrees=45
  swing2Motion=Free
  twistLimitDegrees=45
  twistMotion=Free
 }
 {
  bone1Name=dyn_neck_01
  bone2Name=dyn_spine_04
  swing1LimitDegrees=45
  swing1Motion=Locked
  swing2LimitDegrees=45
  swing2Motion=Locked
  twistLimitDegrees=45
  twistMotion=Locked
 }
 {
  bone1Name=dyn_spine_04
  bone2Name=spine_04
  swing1LimitDegrees=45
  swing1Motion=Free
  swing2LimitDegrees=45
  swing2Motion=Free
  twistLimitDegrees=45
  twistMotion=Free
 }
]
