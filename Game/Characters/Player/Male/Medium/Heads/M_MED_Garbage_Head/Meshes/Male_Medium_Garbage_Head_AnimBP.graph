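# Anim Blueprint for the male medium Garbage player head, adding secondary motion over the pose copied from the body mesh.
# The copied pose gets a neck translation constraint, then two RigidBody passes and AnimDynamics on the dyn_note bones.
# Update derives three alphas from head rotation relative to spine_05, driving additive hose-bone offsets and stretching as corrective adjustments.
# A dyn_spine_04 replace branch exists but does not reach the output pose.
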AnimBlueprint /Game/Characters/Player/Male/Medium/Heads/M_MED_Garbage_Head/Meshes/Male_Medium_Garbage_Head_AnimBP
target_skeleton=&/Male_Medium_Garbage_Head_Skeleton
graphs [
 AnimGraph {nodes [
   "Transform (Modify) Bone - Bone: dyn_spine_04"(
    ComponentPose=Local To Component(LocalPose=Copy Pose From Mesh() {bUseAttachedParent=true})
    Alpha=1
    Translation=Get dyn_spine_04_trans() -> (dyn_spine_04_trans:Vector)
    Rotation=Get dyn_spine_04_rot() -> (dyn_spine_04_rot:Rotator)
    Scale="1,1,1"
   ) {
    rotationMode=BMM_Replace
    rotationSpace=BCS_BoneSpace
    translationMode=BMM_Replace
    translationSpace=BCS_BoneSpace
    alphaBoolBlend {blendOption=Linear}
    boneToModify {boneName=dyn_spine_04}
   }
   n1 = Get Body_Mesh()
   n2 = "Constraint - dyn_neck_01 (T)"(
    ComponentPose=Local To Component(LocalPose=Copy Pose From Mesh(SourceMeshComponent=@n1) {bCopyCurves=true, bUseAttachedParent=true})
    Alpha=1
    ConstraintWeights_0=1
   ) {
    alphaBoolBlend {blendOption=Linear}
    boneToModify {boneName=dyn_neck_01}
    constraintSetup [{transformType=Translation, targetBone {boneName=neck_01}}]
    constraintWeights [1]
   }
   n3 = RigidBody(ComponentPose=@n2) {
    bClampLinearTranslationLimitToRefPose=true
    bOverrideWorldGravity=true
    overridePhysicsAsset=&/Male_Medium_Garbage_Head_Dyn_Boucning
    simulationSpace=BaseBoneSpace
    alphaBoolBlend {blendOption=Linear}
    baseBoneRef {boneName=spine_05}
    componentAppliedLinearAccClamp=<3000 3000 3000>
    componentLinearAccScale=<0.7 0.7 0.7>
    componentLinearVelScale=<0.7 0.7 0.7>
    overrideWorldGravity {z=-980}
   }
   n4 = RigidBody(ComponentPose=@n3) {
    bClampLinearTranslationLimitToRefPose=true
    bOverrideWorldGravity=true
    overridePhysicsAsset=&/Male_Medium_Garbage_Head_perf_Dyn
    simulationSpace=BaseBoneSpace
    alphaBoolBlend {blendOption=Linear}
    baseBoneRef {boneName=spine_05}
    componentAppliedLinearAccClamp=<2000 2000 2000>
    componentLinearAccScale=<0.2 0.1 0.1>
    componentLinearVelScale=<0.2 0.1 0.1>
    overrideWorldGravity {z=-980}
   }
   n5 = Anim Dynamics - Bone: dyn_note_top(ComponentPose=@n4, Alpha=1, ExternalForce="0,0,0") {
    angularDampingOverride=0.998
    angularSpringConstant=300
    bAngularSpring=true
    bOverrideAngularDamping=true
    bUseGravityOverride=true
    simulationSpace=RootRelative
    alphaBoolBlend {blendOption=Linear}
    boundBone {boneName=dyn_note_top}
    componentLinearAccScale=<0.8 0.8 0.8>
    componentLinearVelScale=<0.8 0.8 0.8>
    gravityOverride {z=-540}
    physicsBodyDefinitions [{
      boundBone {boneName=dyn_note_top}
      boxExtents=<7 12 12>
      constraintSetup {angularLimitsMax=<0 0 5>, angularLimitsMin=<0 0 -35>, angularTarget=<0 0 0>}
      localJointOffset=<3 0 0>
     }]
   }
   n6 = Anim Dynamics - Bone: dyn_note_bott(ComponentPose=@n5, Alpha=1, ExternalForce="0,0,0") {
    angularDampingOverride=0.998
    angularSpringConstant=300
    bAngularSpring=true
    bOverrideAngularDamping=true
    bUseGravityOverride=true
    simulationSpace=RootRelative
    alphaBoolBlend {blendOption=Linear}
    boundBone {boneName=dyn_note_bott}
    componentLinearAccScale=<0.8 0.8 0.8>
    componentLinearVelScale=<0.8 0.8 0.8>
    gravityOverride {z=-540}
    physicsBodyDefinitions [{
      boundBone {boneName=dyn_note_bott}
      boxExtents=<7 12 12>
      constraintSetup {angularLimitsMax=<0 0 5>, angularLimitsMin=<0 0 -35>, angularTarget=<0 0 0>}
      localJointOffset=<3 0 0>
     }]
   }
   n7 = Get head rot() -> (head rot:double)
   n8 = "Transform (Modify) Bone - Bone: dyn_hose_04"(ComponentPose=@n6, Alpha=@n7, Translation="2.5,0,0", Rotation="0,0,0", Scale="1,1,1") {
    translationMode=BMM_Additive
    translationSpace=BCS_BoneSpace
    alphaBoolBlend {blendOption=Linear}
    boneToModify {boneName=dyn_hose_04}
    translation {x=2.5}
   }
   n9 = "Transform (Modify) Bone - Bone: dyn_hose_05"(ComponentPose=@n8, Alpha=@n7, Translation="2.5,0,0", Rotation="0,0,0", Scale="1,1,1") {
    translationMode=BMM_Additive
    translationSpace=BCS_BoneSpace
    alphaBoolBlend {blendOption=Linear}
    boneToModify {boneName=dyn_hose_05}
    translation {x=2.5}
   }
   n10 = "Transform (Modify) Bone - Bone: dyn_hose_04"(ComponentPose=@n9, Alpha=Get head rot_2() -> (head rot_2:double), Translation="3,2,0", Rotation="0,0,0", Scale="1.000,1.750,1.000") {
    scaleMode=BMM_Additive
    scaleSpace=BCS_BoneSpace
    translationMode=BMM_Additive
    translationSpace=BCS_BoneSpace
    alphaBoolBlend {blendOption=Linear}
    boneToModify {boneName=dyn_hose_04}
    scale {y=1.75}
    translation=<3 2>
   }
   n11 = Get head rot_3() -> (head rot_3:double)
   n12 = "Transform (Modify) Bone - Bone: dyn_hose_04"(ComponentPose=@n10, Alpha=@n11, Translation="1,1,0", Rotation="0,0,0", Scale="1.000,1.5,1.000") {
    scaleMode=BMM_Additive
    scaleSpace=BCS_BoneSpace
    translationMode=BMM_Additive
    translationSpace=BCS_BoneSpace
    alphaBoolBlend {blendOption=Linear}
    boneToModify {boneName=dyn_hose_04}
    scale {y=1.5}
    translation=<1 1>
   }
   n13 = "Transform (Modify) Bone - Bone: dyn_hose_05"(ComponentPose=@n12, Alpha=@n11, Translation="1,1,0", Rotation="0,0,0", Scale="1.000,1.5,1.000") {
    scaleMode=BMM_Additive
    scaleSpace=BCS_BoneSpace
    translationMode=BMM_Additive
    translationSpace=BCS_BoneSpace
    alphaBoolBlend {blendOption=Linear}
    boneToModify {boneName=dyn_hose_05}
    scale {y=1.5}
    translation=<1 1>
   }
   Output Pose(Result=Component To Local(ComponentPose=@n13)) {layerGroup=None}
   Get PawnVelocityZ()
  ]}
 EventGraph {nodes [
   Event BlueprintUpdateAnimation()
   n1 = Get Skel_Mesh()
   > Set head rot(head rot:double=GetAlphaFromReferenceTransformDelta(SkeletalMeshComponent=@n1, BoneName=head, BaseName=spine_05, InputRange=(X=-20,Y=15), TwistAxis="1,0,0")) -> (Output_Get:double)
   > Set head rot_2(head rot_2:double=GetAlphaFromReferenceTransformDelta(SkeletalMeshComponent=@n1, BoneName=head, BaseName=spine_05, DeltaComponentType=EulerX, InputRange=(X=0,Y=-25), TwistAxis="1,0,0")) -> (Output_Get:double)
   > Set head rot_3(head rot_3:double=GetAlphaFromReferenceTransformDelta(SkeletalMeshComponent=@n1, BoneName=head, BaseName=spine_05, DeltaComponentType=EulerX, InputRange=(X=0,Y=30), TwistAxis="1,0,0")) -> (Output_Get:double)
   Event BlueprintInitializeAnimation()
   n2 = TryGetPawnOwner()
   > Is Valid(InputObject=@n2)
   > n3 = Cast To PlayerPawn_Athena(Object=@n2)
   > n4 = Set Pawn Owner(Pawn Owner:Player Pawn Athena Object Reference=@n3) -> (Output_Get:Player Pawn Athena Object Reference)
   > Set Body_Mesh(Body_Mesh=GetSkeletalMeshForPartType(self=@n4))
   > Set Skel_Mesh(Skel_Mesh=GetOwnerSkeletalMeshComponent())
   n5 = TryGetPawnOwner()
   Is Valid(InputObject=@n5)
   > n6 = Cast To PlayerPawn_Athena(Object=@n5)
   > n7 = Set Pawn Owner(Pawn Owner=@n6)
   n8 = BreakTransform(InTransform=GetDeltaTransformFromRefPose(self=GetSkeletalMeshForPartType(self=@n7), BoneName=dyn_spine_04, BaseName=None))
   > Set dyn_spine_04_rot(dyn_spine_04_rot:Rotator=@n8.Rotation) -> (Output_Get:Rotator)
   > Set dyn_spine_04_trans(dyn_spine_04_trans:Vector=@n8) -> (Output_Get:Vector)
   Event OnCharacterCustomizationLoaded()
  ]}
]
variables [
 Body_Mesh {type=SkeletalMeshComponent}
 dyn_spine_04_rot {type=Rotator}
 dyn_spine_04_trans {type=Vector}
 head rot {type=number}
 head rot_2 {type=number}
 head rot_3 {type=number}
 Pawn Owner {type=PlayerPawn_Athena_C}
 Skel_Mesh {type=SkeletalMeshComponent}
]
